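# Material function for the Baron holographic logo that outputs glitched UVs.
# Two GlitchMap samples, one rotated and scrolling, one time-stepped, build stretched horizontal offsets, which are added to the UVs with a sine wiggle.
# Blueprint-fed start time drives transition timing reroutes, and the FadeOutLogo static switch adds looping SphereMask bursts gated by Blueprint parameters.
# Some branches, including a stepped digital-noise offset, do not reach the output.
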MaterialFunction /Game/Global/Materials/MaterialGym/Materials/HolographicTexture/MF_Glitch_Baron
graph {nodes [
  n1 = Time()
  n2 = Subtract(A=@n1, B=ScalarParameter() {defaultValue=9.999999778196308e+22, parameterName=BP_GetTime})
  n3 = ScalarParameter() {defaultValue=0.52, parameterName=Logo_AppearanceDelay}
  n4 = ScalarParameter() {defaultValue=2, parameterName=Logo_VanishingDelay}
  n5 = Multiply(A=Subtract(A=@n2, B=Add(A=@n3, B=@n4)) {constB=3}) {constB=3}
  Abs(None=Sine(None=Saturate(None=@n5)))
  n6 = Constant()
  n7 = Multiply(A=@n6) {constB=-1}
  n8 = NamedRerouteDeclaration(None=FunctionInput(Preview=TextureCoordinate()) {bUsePreviewValueAsDefault=true, inputName=UV, inputType=FunctionInput_Vector2}) {name=InputUV, nodeColor=<rgba 0 1 0.6953 1>}
  n9 = Multiply(A=Time()) {constB=10}
  n10 = Constant2Vector() {g=-0.3, r=0.7}
  n11 = Multiply(A=@n10) {constB=0.3}
  n12 = NamedRerouteUsage()
  n13 = TextureSample(UVs=Panner(Coordinate=Multiply(A=@n8) {constB=0.45}, Time=Ceil(None=@n9), Speed=LinearInterpolate(A=@n11, B=@n10, Alpha=@n12)) {bFractionalPart=true, speedX=0.7, speedY=-0.3}) {
   samplerSource=SSM_Wrap_WorldGroupSettings
   samplerType=SAMPLERTYPE_LinearColor
   texture=/Game/Global/Textures/Default/Noise/T_Atlas_DigitalNoise_01
  }
  n14 = Constant() {r=0.08}
  Add(
   A=Multiply(A=LinearInterpolate(A=@n7, B=@n6, Alpha=@n13.R) {constA=-0.01, constB=0.01}, B=Constant2Vector() {r=1})
   B=LinearInterpolate(A=Multiply(A=@n14) {constB=-1}, B=@n14, Alpha=@n13.B) {constA=-0.1, constB=0.1}
  )
  n15 = NamedRerouteUsage()
  n16 = Multiply(
   A=Rotator(Coordinate=@n15, Time=ScalarParameter() {group=HolographicTexture_Glitch, parameterName=GlitchUVRotation})
   B=Constant2Vector() {g=10}
  )
  n17 = AppendVector(A=Constant() {r=0.5}, B=ScalarParameter() {defaultValue=5, group=HolographicTexture_Glitch, parameterName=StretchedScrollingSpeed})
  n18 = TextureSample(UVs=Multiply(A=Constant2Vector() {g=0.05, r=1}, B=Panner(Coordinate=@n16, Speed=@n17) {speedX=0.5, speedY=5})) {
   samplerSource=SSM_Wrap_WorldGroupSettings
   samplerType=SAMPLERTYPE_LinearColor
   texture=/Game/Environments/Asteria/Sets/NeonCity/Techart/Holograms/GlitchMap_M
  }
  n19 = Constant2Vector() {g=5, r=1}
  n20 = Floor(None=Multiply(A=Time()) {constB=3})
  n21 = TextureSample(UVs=Add(A=Multiply(A=@n15, B=@n19) {constB=5}, B=Divide(A=@n20) {constB=3})) {
   samplerSource=SSM_Wrap_WorldGroupSettings
   samplerType=SAMPLERTYPE_LinearColor
   texture=&GlitchMap_M
  }
  n22 = Multiply(A=SmoothStep(Value=@n18.R) {constMax=0.1}, B=Saturate(None=Add(A=@n21.G) {constB=0.3}))
  n23 = LinearInterpolate(B=ScalarParameter() {defaultValue=0.1, group=HolographicTexture_Glitch, parameterName=StretchedGlitchIntensity}, Alpha=NamedRerouteUsage()) {constB=0.1}
  n24 = Multiply(B=ScalarParameter() {defaultValue=0.1, group=HolographicTexture_Glitch, parameterName=StretchedGlitchIntensity}) {constA=-0.75}
  n25 = Constant()
  n26 = Add(
   A=Multiply(A=Constant2Vector() {r=1}, B=Multiply(A=@n22, B=@n23) {constB=0.2})
   B=LinearInterpolate(B=AppendVector(A=@n24, B=@n25), Alpha=NamedRerouteUsage()) {constB=0.1}
  )
  n27 = NamedRerouteUsage()
  n28 = Time()
  n29 = Multiply(A=Sine(None=Multiply(A=@n28) {constB=10}), B=Multiply(A=Constant() {desc=Wiggle Intensity, r=0.07}, B=Constant2Vector() {g=0.7, r=1})) {constB=0.007}
  FunctionOutput(None=Add(A=Add(A=@n15, B=@n26), B=Multiply(A=@n27, B=@n29))) {outputName=Glitched UV}
  n30 = Constant2Vector() {g=1.2, r=7.3}
  Multiply(A=Sine(None=Multiply(A=@n1, B=@n30) {constB=7.3}), B=LinearInterpolate(Alpha=NamedRerouteUsage()) {constB=0.05}) {constB=0.05}
  n31 = Multiply(A=@n2, B=ScalarParameter() {defaultValue=1.4, parameterName=GlitchTransition_Speed})
  NamedRerouteDeclaration(None=Saturate(None=Multiply(B=@n31) {constA=2})) {name=GlitchIncrease_Timing, nodeColor=<rgba 1 0 0.9141 1>}
  n32 = Sine(None=Saturate(None=@n31))
  n33 = Saturate(None=Multiply(A=Abs(None=@n32)) {constB=10000})
  n34 = Multiply(A=Subtract(A=@n2, B=@n3), B=Divide(B=Multiply(A=@n4) {constB=2})) {constB=0.3333}
  n35 = Multiply(A=Subtract(A=Frac(None=@n34)) {constB=0.5}) {constB=2}
  n36 = ScalarParameter() {defaultValue=0.04, parameterName=Logo_LoopingGlitch_Duration}
  n37 = Constant2Vector() {r=-1}
  n38 = Add(A=@n36, B=@n37.R)
  n39 = Multiply(A=Add(A=SphereMask(A=@n35, B=@n38, Radius=@n36), B=SphereMask(A=@n35, B=@n37.G, Radius=@n36) {attenuationRadius=0.1})) {constB=10000}
  n40 = ScalarParameter() {parameterName=BP_BaronMaterial_ON/OFF}
  n41 = Multiply(A=Multiply(A=Saturate(None=@n39), B=OneMinus(None=@n40)), B=OneMinus(None=ScalarParameter() {parameterName=BP_StopLoopingLogo}))
  n42 = StaticBool()
  NamedRerouteDeclaration(None=StaticSwitch(
    True=Add(A=@n33, B=@n41)
    False=@n33
    Value=FunctionInput(Preview=@n42) {bUsePreviewValueAsDefault=true, inputName=FadeOutLogo, inputType=FunctionInput_StaticBool}
   )) {name=ActivateGlitch, nodeColor=<rgba 0 0.9062 1 1>}
 ]}
