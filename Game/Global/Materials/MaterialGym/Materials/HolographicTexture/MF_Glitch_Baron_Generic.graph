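# Material function that distorts input UVs into a "Glitched UV" output for holographic textures, with an ActivateGlitch input to drive the effect strength.
# A rotated, stretched, scrolling GlitchMap sample is thresholded and masked by a second sample that jumps on floored time, producing horizontal tearing.
# StretchedGlitchIntensity scales the tearing and a horizontal shift, and a small sine-over-time wiggle is added on both axes.
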MaterialFunction /Game/Global/Materials/MaterialGym/Materials/HolographicTexture/MF_Glitch_Baron_Generic
graph {nodes [
  n1 = NamedRerouteDeclaration(None=FunctionInput(Preview=TextureCoordinate()) {bUsePreviewValueAsDefault=true, inputName=UV, inputType=FunctionInput_Vector2}) {name=InputUV, nodeColor=<rgba 0 1 0.6953 1>}
  n2 = Multiply(
   A=Rotator(Coordinate=@n1, Time=ScalarParameter() {group=HolographicTexture_Glitch, parameterName=GlitchUVRotation})
   B=Constant2Vector() {g=10}
  )
  n3 = AppendVector(A=Constant() {r=0.5}, B=ScalarParameter() {defaultValue=5, group=HolographicTexture_Glitch, parameterName=StretchedScrollingSpeed})
  n4 = TextureSample(UVs=Multiply(A=Constant2Vector() {g=0.1, r=1}, B=Panner(Coordinate=@n2, Speed=@n3) {speedX=0.5, speedY=5})) {
   samplerSource=SSM_Wrap_WorldGroupSettings
   samplerType=SAMPLERTYPE_LinearColor
   texture=/Game/Environments/Asteria/Sets/NeonCity/Techart/Holograms/GlitchMap_M
  }
  n5 = Constant2Vector() {g=5, r=1}
  n6 = Floor(None=Multiply(A=Time()) {constB=3})
  n7 = TextureSample(UVs=Add(A=Multiply(A=@n1, B=@n5) {constB=5}, B=Divide(A=@n6) {constB=3})) {
   samplerSource=SSM_Wrap_WorldGroupSettings
   samplerType=SAMPLERTYPE_LinearColor
   texture=&GlitchMap_M
  }
  n8 = Multiply(A=SmoothStep(Value=@n4.R) {constMax=0.1}, B=Saturate(None=Add(A=@n7.G) {constB=0.3}))
  n9 = LinearInterpolate(B=ScalarParameter() {defaultValue=0.1, group=HolographicTexture_Glitch, parameterName=StretchedGlitchIntensity}, Alpha=NamedRerouteUsage()) {constB=0.1}
  n10 = Multiply(B=ScalarParameter() {defaultValue=0.1, group=HolographicTexture_Glitch, parameterName=StretchedGlitchIntensity}) {constA=-0.75}
  n11 = Constant()
  n12 = Add(
   A=Multiply(A=Constant2Vector() {r=1}, B=Multiply(A=@n8, B=@n9) {constB=0.2})
   B=LinearInterpolate(B=AppendVector(A=@n10, B=@n11), Alpha=NamedRerouteUsage()) {constB=0.1}
  )
  n13 = NamedRerouteUsage()
  n14 = Time()
  n15 = Multiply(A=Sine(None=Multiply(A=@n14) {constB=10}), B=Multiply(A=Constant() {desc=Wiggle Intensity, r=0.01}, B=Constant2Vector() {g=0.7, r=1})) {constB=0.007}
  FunctionOutput(None=Add(A=Add(A=@n1, B=@n12), B=Multiply(A=@n13, B=@n15))) {outputName=Glitched UV}
  n16 = FunctionInput() {bUsePreviewValueAsDefault=true, inputName=ActivateGlitch, inputType=FunctionInput_Scalar, sortPriority=1, previewValue=<1 0 0 1>}
  NamedRerouteDeclaration(None=@n16) {name=ActivateGlitch, nodeColor=<rgba 0 0.9062 1 1>}
 ]}
