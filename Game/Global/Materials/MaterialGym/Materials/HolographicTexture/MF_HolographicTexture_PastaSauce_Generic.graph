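# Material function that overlays a tinted hologram texture with scanlines onto incoming material attributes, also outputting emissive and opacity.
# It samples the texture through a screen-glitch UV function as a jittering bottom layer and an offset top layer, with separate tints.
# Masks blend the hologram over the original, and static switches toggle texture-only mode, base color override and noise-driven normal-direction WPO glitching.
# A grid/stretch UV glitch chain is present but feeds no output.
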MaterialFunction /Game/Global/Materials/MaterialGym/Materials/HolographicTexture/MF_HolographicTexture_PastaSauce_Generic
graph {nodes [
  n1 = MaterialFunctionCall!(
   ScanLine_Tiling_1
   Scanline_PannerSpeed_1
   ScanLine_Tiling
   Scanline_PannerSpeed
   Scanline_Rotation
   UVs=FunctionInput(Preview=TextureCoordinate()) {bUsePreviewValueAsDefault=true, inputName=Scanlines_UV, inputType=FunctionInput_Vector2, sortPriority=2}
  ) {desc=ScanLine, materialFunction=&/MF_HolographicTexture_ScanLines}
  n2 = Saturate(None=Add(A=@n1.Result, B=ScalarParameter() {defaultValue=0.5, group=HolographicTexture, parameterName=Scanline_ColorMultiplier, sortPriority=0}) {constB=0})
  n3 = FunctionInput(Preview=TextureCoordinate()) {bUsePreviewValueAsDefault=true, inputName=TextureUV, inputType=FunctionInput_Vector2, sortPriority=2}
  n4 = MaterialFunctionCall!(Main_TextureUV=@n3) {materialFunction=/Game/Global/Materials/MaterialGym/Materials/ScreenGlitchGeneric/MF_ScreenGlitch_Generic}
  n5 = Multiply(A=Time()) {constB=7}
  n6 = Add(A=Multiply(A=Sine(None=@n5)) {constB=0.5}) {constB=0.5}
  n7 = TextureSampleParameter2D(UVs=Add(A=@n4.Result, B=LinearInterpolate(Alpha=@n6) {constA=-0.01, constB=0})) {
   group=HolographicTexture
   parameterName=HologramTexture
   samplerSource=SSM_Clamp_WorldGroupSettings
   sortPriority=-20
   texture=/Game/Accessories/FORT_Backpacks/Backpack_F_MED_PastaSauce/Textures/SelectableSymbols/T_PastaSauce_Backpack_SelectableSymbol_03
  }
  n8 = VectorParameter() {desc=HOLOGRAM BASE COLOR, group=HolographicTexture, parameterName=HologramBaseColor, sortPriority=-4, defaultValue=<rgba 0.1357 0.2884 1 1>}
  n9 = FunctionInput(Preview=@n8.RGB) {bUsePreviewValueAsDefault=true, inputName=HologramBaseColor}
  n10 = VectorParameter() {group=HolographicTexture, parameterName=BottomTextureTint, sortPriority=1000, defaultValue=<rgba 1 1 1 1>}
  n11 = Multiply(
   A=Multiply(A=@n7.RGB, B=ScalarParameter() {defaultValue=1, group=HolographicTexture, parameterName=TextureBrightness, sortPriority=-19})
   B=@n10.RGB
  )
  n12 = StaticSwitchParameter(
   True=Multiply(A=@n7.RGB, B=@n7.RGB)
   False=StaticSwitchParameter(True=LinearInterpolate(A=@n9, B=@n11, Alpha=@n7.A), False=@n9) {defaultValue=true, group=HolographicTexture, parameterName=UseHologramTexture, sortPriority=-5}
  ) {group=HolographicTexture, parameterName=UseTextureOnly, sortPriority=-5}
  n13 = ChannelMaskParameter(None=NamedRerouteUsage()) {group=HolographicTexture, maskChannel=Alpha, parameterName=TextureMask_Channel, sortPriority=-10, defaultValue=<rgba 0 0 0 1>}
  n14 = TextureSampleParameter2D() {
   group=HolographicTexture
   parameterName=HologramMask
   sortPriority=-10
   texture=/Game/Weapons/FORT_Melee/Pickaxe_CosmicSquatter/Textures/T_Pickaxe_CosmicSquatter_FX
  }
  n15 = ChannelMaskParameter(None=@n14.RGB) {group=HolographicTexture, parameterName=HologramMask_Channel, sortPriority=-10}
  n16 = StaticSwitchParameter(
   True=@n13.R
   False=StaticSwitchParameter(
    True=FunctionInput(Preview=@n15.R) {bUsePreviewValueAsDefault=true, inputName=HologramMask, inputType=FunctionInput_Scalar, sortPriority=5}
    False=Constant() {r=1}
   ) {group=HolographicTexture, parameterName=UseHologramMask, sortPriority=-10}
  ) {group=HolographicTexture, parameterName=UseTextureOnly, sortPriority=-5}
  n17 = Add(B=ScalarParameter() {defaultValue=0.5, group=HolographicTexture, parameterName=Scanline_ColorMultiplier, sortPriority=0}) {constB=0}
  n18 = ScalarParameter() {defaultValue=0.1, group=HolographicTexture, parameterName=TopTexture_Offset_X, sortPriority=1000}
  n19 = ScalarParameter() {defaultValue=0.1, group=HolographicTexture, parameterName=TopTexture_Offset_Y, sortPriority=1000}
  n20 = TextureSampleParameter2D(UVs=Add(A=@n4.Result, B=AppendVector(A=@n18, B=@n19))) {
   group=HolographicTexture
   parameterName=HologramTexture
   samplerSource=SSM_Clamp_WorldGroupSettings
   sortPriority=-20
   texture=&T_PastaSauce_Backpack_SelectableSymbol_03
  }
  n21 = Multiply(A=@n20.RGB, B=@n20.RGB)
  n22 = VectorParameter() {desc=HOLOGRAM BASE COLOR, group=HolographicTexture, parameterName=HologramBaseColor, sortPriority=-4, defaultValue=<rgba 0.1357 0.2884 1 1>}
  n23 = FunctionInput(Preview=@n22.RGB) {bUsePreviewValueAsDefault=true, inputName=HologramBaseColor_1}
  n24 = ScalarParameter() {defaultValue=1, group=HolographicTexture, parameterName=TextureBrightness_TopColor, sortPriority=-19}
  n25 = Multiply(
   A=@n20.RGB
   B=MaterialFunctionCall(Value_Morning=@n24, Value_Day=@n24, Value_Evening=@n24, Value_Night=Multiply(A=@n24) {constB=0.7}) {materialFunction=&Scalar_Time_of_Day}.Result
  )
  n26 = VectorParameter() {group=HolographicTexture, parameterName=TopTextureTint, sortPriority=1000, defaultValue=<rgba 1 1 1 1>}
  n27 = StaticSwitchParameter(True=LinearInterpolate(A=@n23, B=Multiply(A=@n25, B=@n26.RGB), Alpha=@n20.A), False=@n23) {defaultValue=true, group=HolographicTexture, parameterName=UseHologramTexture, sortPriority=-5}
  n28 = LinearInterpolate(
   A=LinearInterpolate(A=NamedRerouteUsage(), B=Multiply(A=@n2, B=@n12), Alpha=@n16)
   B=Multiply(
    A=Saturate(None=@n17)
    B=StaticSwitchParameter(True=@n21, False=@n27) {group=HolographicTexture, parameterName=UseTextureOnly, sortPriority=-5}
   )
   Alpha=@n20.A
  )
  n29 = Add(A=@n1.Result, B=ScalarParameter() {defaultValue=0.8, group=HolographicTexture, parameterName=Scanline_Opacity, sortPriority=0}) {constB=0.8}
  n30 = LinearInterpolate(
   A=NamedRerouteUsage()
   B=Multiply(A=ScalarParameter() {defaultValue=1, group=HolographicTexture, parameterName=HologramOpacity, sortPriority=-10}, B=Saturate(None=@n29))
   Alpha=@n16
  )
  n31 = NamedRerouteUsage()
  n32 = Constant()
  n33 = Panner(Coordinate=TextureCoordinate()) {speedX=5, speedY=10}
  n34 = ScalarParameter() {defaultValue=1, group=HolographicTexture_Glitch, parameterName=WPO_GlitchNoiseFrequency, sortPriority=33}
  n35 = Multiply(
   A=NamedRerouteUsage()
   B=TextureSample(UVs=Multiply(A=@n33, B=@n34)) {samplerType=SAMPLERTYPE_Masks, texture=/Game/Characters/Player/Male/Medium/Bodies/M_MED_Storm_Soldier/Textures/ElectricityNoise_M}.RGB
  )
  n36 = MaterialFunctionCall(Color=VertexNormalWS()) {materialFunction=&SplitComponents}
  n37 = Multiply(
   A=MaterialFunctionCall(R=@n36.R, G=@n36.R, B=Multiply(A=@n36.B) {constB=0.3}) {materialFunction=&AppendMany}.RGB
   B=ScalarParameter() {defaultValue=35, group=HolographicTexture_Glitch, parameterName=WPO_GlitchIntensity, sortPriority=33}
  ) {constB=35}
  n38 = SetMaterialAttributes!(
   MaterialAttributes=NamedRerouteUsage()
   EmissiveColor=@n28
   Opacity=@n30
   BaseColor=StaticSwitchParameter(True=LinearInterpolate(A=@n31, B=@n32, Alpha=@n16), False=@n31) {defaultValue=true, group=HolographicTexture, parameterName=OverrideBaseColorOnHologram, sortPriority=-100}
   WorldPositionOffset=StaticSwitchParameter(True=Multiply(A=@n35, B=@n37) {constB=100}, False=NamedRerouteUsage()) {group=HolographicTexture_Glitch, parameterName=UseGlitchWPO, sortPriority=33}
  ) {attributeSetTypes [
    #1
    #2
    #3
    #4
   ]}
  FunctionOutput(None=@n38) {outputName=MaterialAttributes, sortPriority=-1}
  FunctionOutput(None=@n28) {outputName=EmissiveColor, sortPriority=1}
  FunctionOutput(None=@n30) {outputName=Opacity}
  n39 = Constant() {r=1}
  n40 = Subtract(A=@n39, B=ScalarParameter() {defaultValue=0.075, group=HolographicTexture_Glitch, parameterName=GlitchDuration, sortPriority=10})
  n41 = Multiply(A=Time(), B=ScalarParameter() {defaultValue=0.2, group=HolographicTexture_Glitch, parameterName=GlitchFrequency, sortPriority=10}) {constB=0.2}
  n42 = Add(A=Multiply(A=Sine(None=@n41)) {constB=0.5}) {constB=0.5}
  NamedRerouteDeclaration(None=Multiply(
    A=SmoothStep(Min=@n40, Max=@n39, Value=@n42) {constMax=0.9, constMin=0.8}
    B=MaterialFunctionCall() {materialFunction=/Game/MF__FloatFlickerRandom_Generic}.RandomFlickerFloat
   )) {name=GlitchFrequencyMask, nodeColor=<rgba 0.9377 1 0.8953 1>}
  NamedRerouteDeclaration(None=@n7.RGBA) {name=Texture RGBA, nodeColor=<rgba 1 0 0.8906 1>}
  n43 = GetMaterialAttributes(None=FunctionInput() {inputName=MaterialAttributes, inputType=FunctionInput_MaterialAttributes, sortPriority=-1}) {attributeGetTypes [
    #3
    #4
    #1
    #2
   ]}
  NamedRerouteDeclaration(None=@n43.BaseColor) {name=OG Base Color, nodeColor=<rgba 0 0.7422 1 1>}
  NamedRerouteDeclaration(None=@n43.WorldPositionOffset) {name=OG WPO, nodeColor=<rgba 0 1 0.6016 1>}
  NamedRerouteDeclaration(None=@n43.EmissiveColor) {name=OG Emissive, nodeColor=<rgba 0 1 0.1797 1>}
  NamedRerouteDeclaration(None=@n43.Opacity) {name=OG Opacity, nodeColor=<rgba 0.3422 0.8799 1 1>}
  NamedRerouteDeclaration(None=@n43.MaterialAttributes) {name=OG Material Attributes, nodeColor=<rgba 0.8984 0 1 1>}
  n44 = LinearInterpolate(
   A=@n3
   B=MaterialFunctionCall!(UV=@n3) {materialFunction=&/MF_HolographicTextureGlitch_Grid}.Glitched UV
   Alpha=NamedRerouteUsage()
  )
  n45 = Constant2Vector() {r=1}
  n46 = Multiply(
   A=Rotator(Coordinate=TextureCoordinate(), Time=ScalarParameter() {group=HolographicTexture_Glitch, parameterName=GlitchUVRotation})
   B=Constant2Vector() {g=10}
  )
  n47 = AppendVector(A=Constant() {r=0.5}, B=ScalarParameter() {defaultValue=5, group=HolographicTexture_Glitch, parameterName=StretchedScrollingSpeed})
  n48 = TextureSample(UVs=Multiply(A=Constant2Vector() {g=0.1, r=1}, B=Panner(Coordinate=@n46, Speed=@n47) {speedX=0.5, speedY=5})) {samplerType=SAMPLERTYPE_LinearColor, texture=/Game/Environments/Asteria/Sets/NeonCity/Techart/Holograms/GlitchMap_M}
  n49 = Multiply(A=TextureCoordinate(), B=Constant2Vector() {g=7, r=1}) {constB=5}
  n50 = Time()
  n51 = Divide(A=Floor(None=Multiply(A=@n50) {constB=3})) {constB=3}
  n52 = Add(A=TextureSample(UVs=Add(A=@n49, B=@n51)) {samplerType=SAMPLERTYPE_LinearColor, texture=&GlitchMap_M}.G) {constB=0.3}
  n53 = Multiply(
   A=Multiply(A=SmoothStep(Value=@n48.R) {constMax=0.1}, B=Saturate(None=@n52))
   B=LinearInterpolate(B=ScalarParameter() {defaultValue=0.1, group=HolographicTexture_Glitch, parameterName=StretchedGlitchIntensity}, Alpha=NamedRerouteUsage()) {constB=0.1}
  ) {constB=0.2}
  n54 = AppendVector(
   A=Multiply(B=ScalarParameter() {defaultValue=0.1, group=HolographicTexture_Glitch, parameterName=StretchedGlitchIntensity}) {constA=-0.75}
   B=Constant()
  )
  n55 = NamedRerouteUsage()
  n56 = Add(A=@n3, B=Add(A=Multiply(A=@n45, B=@n53), B=LinearInterpolate(B=@n54, Alpha=@n55) {constB=0.1}))
  n57 = NamedRerouteUsage()
  n58 = Time()
  n59 = Multiply(A=Sine(None=Multiply(A=@n58) {constB=10}), B=Multiply(A=Constant() {desc=Wiggle Intensity, r=0.0075}, B=Constant2Vector() {g=1, r=0.2})) {constB=0.007}
  StaticSwitchParameter(
   True=Add(
    A=StaticSwitchParameter(True=@n44, False=@n56) {defaultValue=true, group=HolographicTexture_Glitch, parameterName=UseGridGlitch, sortPriority=0}
    B=Multiply(A=@n57, B=@n59)
   )
   False=@n3
  ) {defaultValue=true, group=HolographicTexture_Glitch, parameterName=UseGlitch, sortPriority=-1}
 ]}
